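# Physics asset for the necklace chain and pendant of the medium male Vigilante player body.
# Box bodies cover the chain, pendant and pelvis, while spine_05 combines a capsule and an angled box around the upper torso.
# The only constraint joins the chain to spine_05 with tight twist and second-swing limits and a wider first swing.
# The pendant and pelvis bodies have no constraint in this asset.
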PhysicsAsset /BRCosmetics/Characters/Player/Male/Medium/Bodies/M_MED_Vigilante/Meshes/M_MED_Vigilante_Chain_Physics
bodies {
 dyn_necklace_chain [{shapeType=Box, extentX=11.6297, extentY=2.0717, extentZ=9.8143, center=<6.4137 -0.0118 0.2696>, rotation=<rot -0.0771 179.8214 -91.7916>}]
 dyn_necklace_pendant [{shapeType=Box, extentX=5.8534, extentY=3.6978, extentZ=1.3588, center=<2.983 0 0.9326>, rotation=<rot 11.2303 0 0>}]
 pelvis [{shapeType=Box, extentX=7.5, extentY=7.5, extentZ=7.5}]
 spine_05 [
  {shapeType=Capsule, length=17.2812, radius=12.7188, center=<-8.5256 2.371 0>, rotation=<rot 0 3.8784 0>}
  {shapeType=Box, extentX=19.5, extentY=7.5, extentZ=22, center=<-1.6741 7.8579 -1.0043>, rotation=<rot 0 -34.5226 0>}
 ]
}
constraints [{bone1Name=dyn_necklace_chain, bone2Name=spine_05, swing1LimitDegrees=25, swing2LimitDegrees=3, twistLimitDegrees=2}]
