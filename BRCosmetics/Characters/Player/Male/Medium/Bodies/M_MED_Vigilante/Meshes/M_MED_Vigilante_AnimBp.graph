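# Post-process style AnimBlueprint for the M_MED_Vigilante player body, copying the attached parent's pose and adding secondary motion to jacket, zippers, chain, straps and collar.
# Additive bone tweaks scale the left thigh when crouching and shift collar bones using head rotation, targeting and jump/fall state.
# AnimDynamics, weighted constraints and a chain RigidBody simulate loose parts.
# The EventGraph computes blend alphas and a delayed upward external force.
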AnimBlueprint /BRCosmetics/Characters/Player/Male/Medium/Bodies/M_MED_Vigilante/Meshes/M_MED_Vigilante_AnimBp
target_skeleton=&/M_MED_Vigilante_Skeleton
graphs [
 AnimGraph {nodes [
   n1 = "Transform (Modify) Bone - Bone: dyn_thigh_l_scale"(
    ComponentPose=Local To Component(LocalPose=Copy Pose From Mesh() {bUseAttachedParent=true})
    Alpha=Get crouch_leg_volume() -> (crouch_leg_volume:double)
    Translation="0,0,0"
    Rotation="0,0,0"
    Scale="1.3,1.3,1.3"
   ) {
    scaleMode=BMM_Additive
    scaleSpace=BCS_BoneSpace
    alphaBoolBlend {blendOption=Linear}
    boneToModify {boneName=dyn_thigh_l_scale}
    scale=<1.3 1.3 1.3>
   }
   n2 = Get bIsFalling()
   n3 = Get bIsJumping()
   n4 = "Transform (Modify) Bone - Bone: dyn_collar_right"(
    ComponentPose=@n1
    Alpha=SelectFloat(A=Get rotation2() -> (rotation2:double), B=0, bPickA=OR Boolean(A=@n3, B=@n2))
    Translation="-2,0,0"
    Rotation="-6,0,0"
    Scale="1,1,1"
   ) {
    rotationMode=BMM_Additive
    rotationSpace=BCS_BoneSpace
    translationMode=BMM_Additive
    translationSpace=BCS_BoneSpace
    alphaBoolBlend {blendOption=Linear}
    boneToModify {boneName=dyn_collar_right}
    rotation {pitch=-6}
    translation {x=-2}
   }
   n5 = Get head_rot_back() -> (head_rot_back:double)
   n6 = SelectFloat(A=Get ads_rotate() -> (ads_rotate:double), B=@n5, bPickA=Get bIsTargeting())
   n7 = "Transform (Modify) Bone - Bone: dyn_collar_left"(ComponentPose=@n4, Alpha=@n6, Translation="0,0,1", Rotation="2,0,0", Scale="1,1,1") {
    rotationMode=BMM_Additive
    rotationSpace=BCS_BoneSpace
    translationMode=BMM_Additive
    translationSpace=BCS_BoneSpace
    alphaBoolBlend {blendOption=Linear}
    boneToModify {boneName=dyn_collar_left}
    rotation {pitch=2}
    translation {z=1}
   }
   n8 = "Transform (Modify) Bone - Bone: dyn_collar_right"(ComponentPose=@n7, Alpha=@n6, Translation="0,-2,0", Rotation="0,0,-7", Scale="1,1,1") {
    rotationMode=BMM_Additive
    rotationSpace=BCS_BoneSpace
    translationMode=BMM_Additive
    translationSpace=BCS_BoneSpace
    alphaBoolBlend {blendOption=Linear}
    boneToModify {boneName=dyn_collar_right}
    rotation {roll=-7}
    translation {y=-2}
   }
   n9 = "Transform (Modify) Bone - Bone: dyn_collar_back"(ComponentPose=@n8, Alpha=@n6, Translation="0,0,0", Rotation="0,-5,-10", Scale="1,1,1") {
    rotationMode=BMM_Additive
    rotationSpace=BCS_BoneSpace
    translationMode=BMM_Additive
    translationSpace=BCS_BoneSpace
    alphaBoolBlend {blendOption=Linear}
    boneToModify {boneName=dyn_collar_back}
    rotation {roll=-10, yaw=-5}
   }
   n10 = "Transform (Modify) Bone - Bone: dyn_collar_back"(ComponentPose=@n9, Alpha=@n5, Translation="0,1,0", Rotation="0,0,-35", Scale="1,1,1") {
    rotationMode=BMM_Additive
    rotationSpace=BCS_BoneSpace
    translationMode=BMM_Additive
    translationSpace=BCS_BoneSpace
    alphaBoolBlend {blendOption=Linear}
    boneToModify {boneName=dyn_collar_back}
    rotation {roll=-35}
    translation {y=1}
   }
   n11 = "Transform (Modify) Bone - Bone: dyn_collar_right"(ComponentPose=@n10, Alpha=@n5, Translation="0,0,0", Rotation="10,0,-5", Scale="1,1,1") {
    rotationMode=BMM_Additive
    rotationSpace=BCS_BoneSpace
    translationMode=BMM_Additive
    translationSpace=BCS_BoneSpace
    alphaBoolBlend {blendOption=Linear}
    boneToModify {boneName=dyn_collar_right}
    rotation {pitch=10, roll=-5}
   }
   n12 = "Transform (Modify) Bone - Bone: dyn_collar_left"(ComponentPose=@n11, Alpha=@n5, Translation="0,0,0", Rotation="10,0,-20", Scale="1,1,1") {
    rotationMode=BMM_Additive
    rotationSpace=BCS_BoneSpace
    translationMode=BMM_Additive
    translationSpace=BCS_BoneSpace
    alphaBoolBlend {blendOption=Linear}
    boneToModify {boneName=dyn_collar_left}
    rotation {pitch=10, roll=-20}
   }
   n13 = Get ext_force() -> (ext_force:Vector)
   n14 = Anim Dynamics - Bone: dyn_jacket_back(ComponentPose=@n12, Alpha=1, ExternalForce=@n13) {
    angularDampingOverride=0.99
    bGravityOverrideInSimSpace=true
    bOverrideAngularDamping=true
    bOverrideLinearDamping=true
    gravityScale=0.5
    linearDampingOverride=0.99
    alphaBoolBlend {blendOption=Linear}
    boundBone {boneName=dyn_jacket_back}
    componentAppliedLinearAccClamp=<400 400 400>
    componentLinearAccScale=<0.25 0.25 0.25>
    componentLinearVelScale=<0.25 0.25 0.25>
    physicsBodyDefinitions [{
      sphereCollisionRadius=0
      boundBone {boneName=dyn_jacket_back}
      boxExtents=<8 16 2>
      constraintSetup {angularLimitsMax=<0 50 15>, angularLimitsMin=<0 -50 -15>, angularTarget=<0 0 0>}
      localJointOffset=<2 0 0>
     }]
    relativeSpaceBone {boneName=pelvis}
   }
   n15 = Anim Dynamics - Bone: dyn_jacket_r(ComponentPose=@n14, Alpha=1, ExternalForce=@n13) {
    angularDampingOverride=0.999
    bGravityOverrideInSimSpace=true
    bOverrideAngularDamping=true
    bOverrideLinearDamping=true
    linearDampingOverride=0.999
    alphaBoolBlend {blendOption=Linear}
    boundBone {boneName=dyn_jacket_r}
    componentAppliedLinearAccClamp=<1000 1000 1000>
    componentLinearAccScale=<0.5 0.5 0.5>
    componentLinearVelScale=<0.5 0.5 0.5>
    physicsBodyDefinitions [{
      sphereCollisionRadius=0
      boundBone {boneName=dyn_jacket_r}
      boxExtents=<9 16 3>
      constraintSetup {angularLimitsMax=<0 50 10>, angularLimitsMin=<0 -50 -10>, angularTarget=<0 0 0>}
      localJointOffset=<2 0 0>
     }]
    relativeSpaceBone {boneName=root}
   }
   n16 = Anim Dynamics - Bone: dyn_jacket_l(ComponentPose=@n15, Alpha=1, ExternalForce=@n13) {
    angularDampingOverride=0.999
    bGravityOverrideInSimSpace=true
    bOverrideAngularDamping=true
    bOverrideLinearDamping=true
    linearDampingOverride=0.999
    alphaBoolBlend {blendOption=Linear}
    boundBone {boneName=dyn_jacket_l}
    componentAppliedLinearAccClamp=<1000 1000 1000>
    componentLinearAccScale=<0.5 0.5 0.5>
    componentLinearVelScale=<0.5 0.5 0.5>
    physicsBodyDefinitions [{
      sphereCollisionRadius=0
      boundBone {boneName=dyn_jacket_l}
      boxExtents=<9 16 3>
      constraintSetup {angularLimitsMax=<0 50 5>, angularLimitsMin=<0 -50 -5>, angularTarget=<0 0 0>}
      localJointOffset=<2 0 0>
     }]
    relativeSpaceBone {boneName=root}
   }
   n17 = "Constraint - dyn_zipper_jacket (PPPPPPP)"!(
    ComponentPose=@n16
    Alpha=1
    ConstraintWeights_0=0.4136
    ConstraintWeights_1:float=0.015
    ConstraintWeights_2:float=0.0484
    ConstraintWeights_3:float=0.017
    ConstraintWeights_4:float=0.0117
    ConstraintWeights_5:float=0.001
    ConstraintWeights_6:float=0.4933
   ) -> (Pose) {
    alphaBoolBlend {blendOption=Linear}
    boneToModify {boneName=dyn_zipper_jacket}
    constraintSetup [
     {targetBone {boneName=dyn_jacket_r}}
     {targetBone {boneName=thigh_twist_01_r}}
     {targetBone {boneName=spine_01}}
     {targetBone {boneName=spine_02}}
     {targetBone {boneName=spine_03}}
     {targetBone {boneName=spine_04}}
     {targetBone {boneName=pelvis}}
    ]
    constraintWeights [1,1,1,1,1,1,1]
   }
   n18 = "Constraint - dyn_top_zipper_jacket (PPPPPPPPP)"!(
    ComponentPose=@n17
    Alpha=1
    ConstraintWeights_0=0.0206
    ConstraintWeights_1:float=0.0673
    ConstraintWeights_2:float=0.0846
    ConstraintWeights_3:float=0.0138
    ConstraintWeights_4:float=0.0626
    ConstraintWeights_5:float=0.12
    ConstraintWeights_6:float=0.1406
    ConstraintWeights_7:float=0.0944
    ConstraintWeights_8:float=0.3962
   ) -> (Pose) {
    alphaBoolBlend {blendOption=Linear}
    boneToModify {boneName=dyn_top_zipper_jacket}
    constraintSetup [
     {targetBone {boneName=upperarm_r}}
     {targetBone {boneName=upperarm_twist_01_r}}
     {targetBone {boneName=upperarm_twist_02_r}}
     {targetBone {boneName=spine_01}}
     {targetBone {boneName=spine_02}}
     {targetBone {boneName=spine_03}}
     {targetBone {boneName=spine_04}}
     {targetBone {boneName=spine_05}}
     {targetBone {boneName=clavicle_r}}
    ]
    constraintWeights [1,1,1,1,1,1,1,1,1]
   }
   n19 = Anim Dynamics - Bone: dyn_zipper_jacket(ComponentPose=@n18, Alpha=1, ExternalForce="0,0,0") {
    bGravityOverrideInSimSpace=true
    bOverrideLinearDamping=true
    gravityScale=0.25
    linearDampingOverride=0.999
    alphaBoolBlend {blendOption=Linear}
    boundBone {boneName=dyn_zipper_jacket}
    componentAppliedLinearAccClamp=<7000 7000 7000>
    componentLinearAccScale=<0.5 0.5 0.5>
    componentLinearVelScale=<0.5 0.5 0.5>
    physicsBodyDefinitions [{
      sphereCollisionRadius=0
      boundBone {boneName=dyn_zipper_jacket}
      boxExtents=<6 4 2>
      constraintSetup {angularLimitsMax=<20 0 20>, angularLimitsMin=<-20 0 -20>, angularTarget=<0 0 0>}
      localJointOffset=<2 0 0>
     }]
   }
   n20 = Anim Dynamics - Bone: dyn_top_zipper_jacket(ComponentPose=@n19, Alpha=1, ExternalForce="0,0,0") {
    bGravityOverrideInSimSpace=true
    gravityScale=0.25
    alphaBoolBlend {blendOption=Linear}
    boundBone {boneName=dyn_top_zipper_jacket}
    componentAppliedLinearAccClamp=<7000 7000 7000>
    componentLinearAccScale=<0.5 0.5 0.5>
    componentLinearVelScale=<0.5 0.5 0.5>
    physicsBodyDefinitions [{
      sphereCollisionRadius=0
      boundBone {boneName=dyn_top_zipper_jacket}
      boxExtents=<6 4 2>
      constraintSetup {angularLimitsMax=<0 10 10>, angularLimitsMin=<0 -10 -10>, angularTarget=<0 0 0>}
      localJointOffset=<2 0 0>
     }]
   }
   n21 = RigidBody(ComponentPose=@n20) {
    bClampLinearTranslationLimitToRefPose=true
    bOverrideWorldGravity=true
    overridePhysicsAsset=&/M_MED_Vigilante_Chain_Physics
    simulationSpace=BaseBoneSpace
    alphaBoolBlend {blendOption=Linear}
    baseBoneRef {boneName=spine_02}
    componentAppliedLinearAccClamp=<1000 1000 1000>
    componentLinearAccScale=<0.75 0.75 0.75>
    componentLinearVelScale=<0.75 0.75 0.75>
    overrideWorldGravity {z=-980}
   }
   n22 = "Constraint - const_chain_base (PPPPPP)"!(
    ComponentPose=@n21
    Alpha=1
    ConstraintWeights_0=0.0086
    ConstraintWeights_1:float=0.2549
    ConstraintWeights_2:float=0.0258
    ConstraintWeights_3:float=0.0261
    ConstraintWeights_4:float=0.0012
    ConstraintWeights_5:float=0.6835
   ) -> (Pose) {
    alphaBoolBlend {blendOption=Linear}
    boneToModify {boneName=const_chain_base}
    constraintSetup [
     {targetBone {boneName=thigh_r}}
     {targetBone {boneName=thigh_twist_01_r}}
     {targetBone {boneName=spine_01}}
     {targetBone {boneName=thigh_twist_01_l}}
     {targetBone {boneName=thigh_l}}
     {targetBone {boneName=pelvis}}
    ]
    constraintWeights [1,1,1,1,1,1]
   }
   n23 = Anim Dynamics - Bone: dyn_chain_parent(ComponentPose=@n22, Alpha=1, ExternalForce=Get ext_force()) {
    bGravityOverrideInSimSpace=true
    bOverrideLinearDamping=true
    linearDampingOverride=1
    alphaBoolBlend {blendOption=Linear}
    boundBone {boneName=dyn_chain_parent}
    componentAppliedLinearAccClamp=<500 500 500>
    componentLinearAccScale=<0.5 0.5 0.5>
    componentLinearVelScale=<0.75 0.75 0.75>
    physicsBodyDefinitions [{
      sphereCollisionRadius=0
      boundBone {boneName=dyn_chain_parent}
      boxExtents=<15 10 6>
      constraintSetup {angularLimitsMax=<0 1 0>, angularLimitsMin=<0 -60 0>, angularTarget=<0 0 0>}
      localJointOffset=<4 0 0>
     }]
    relativeSpaceBone {boneName=pelvis}
    sphericalLimits [{limitRadius=10, drivingBone {boneName=thigh_twist_01_r}, sphereLocalOffset=<-13 0 -3>}]
   }
   n24 = "Transform (Modify) Bone - Bone: dyn_chain_parent"(ComponentPose=@n23, Alpha=Get chain_alpha() -> (chain_alpha:double), Translation="0,0,2", Rotation="7,0,0", Scale="1,1,1") {
    rotationMode=BMM_Additive
    rotationSpace=BCS_BoneSpace
    translationMode=BMM_Additive
    translationSpace=BCS_BoneSpace
    alphaBoolBlend {blendOption=Linear}
    boneToModify {boneName=dyn_chain_parent}
    rotation {pitch=7}
    translation {z=2}
   }
   n25 = Get ext_force()
   n26 = Anim Dynamics - Bone: dyn_bottom_l_strap(ComponentPose=@n24, Alpha=1, ExternalForce=@n25) {
    bGravityOverrideInSimSpace=true
    simulationSpace=Actor
    alphaBoolBlend {blendOption=Linear}
    boundBone {boneName=dyn_bottom_l_strap}
    componentAppliedLinearAccClamp=<200 200 200>
    componentLinearAccScale=<0.25 0.25 0.25>
    componentLinearVelScale=<0.25 0.25 0.25>
    physicsBodyDefinitions [{
      sphereCollisionRadius=0
      boundBone {boneName=dyn_bottom_l_strap}
      boxExtents=<20 15 3>
      constraintSetup {twistAxis=AxisY, angularLimitsMax=<30 30 10>, angularLimitsMin=<-30 -30 -10>, angularTarget=<0 0 0>}
      localJointOffset=<8 0 0>
     }]
    relativeSpaceBone {boneName=pelvis}
   }
   n27 = Anim Dynamics - Bone: dyn_bottom_r_strap(ComponentPose=@n26, Alpha=1, ExternalForce=@n25) {
    bGravityOverrideInSimSpace=true
    simulationSpace=Actor
    alphaBoolBlend {blendOption=Linear}
    boundBone {boneName=dyn_bottom_r_strap}
    componentAppliedLinearAccClamp=<200 200 200>
    componentLinearAccScale=<0.25 0.25 0.25>
    componentLinearVelScale=<0.25 0.25 0.25>
    physicsBodyDefinitions [{
      sphereCollisionRadius=0
      boundBone {boneName=dyn_bottom_r_strap}
      boxExtents=<20 15 3>
      constraintSetup {twistAxis=AxisY, angularLimitsMax=<30 30 20>, angularLimitsMin=<-30 -30 -20>, angularTarget=<0 0 0>}
      localJointOffset=<8 0 0>
     }]
   }
   n28 = Anim Dynamics - Bone: dyn_top_l_strap(ComponentPose=@n27, Alpha=1, ExternalForce=@n25) {
    bGravityOverrideInSimSpace=true
    alphaBoolBlend {blendOption=Linear}
    boundBone {boneName=dyn_top_l_strap}
    componentAppliedLinearAccClamp=<500 500 500>
    componentLinearAccScale=<0.75 0.75 0.75>
    componentLinearVelScale=<0.75 0.75 0.75>
    physicsBodyDefinitions [{
      sphereCollisionRadius=0
      boundBone {boneName=dyn_top_l_strap}
      boxExtents=<8 3 3>
      constraintSetup {angularLimitsMax=<10 10 20>, angularLimitsMin=<-10 -10 -20>, angularTarget=<0 0 0>}
      localJointOffset=<2 0 0>
     }]
    relativeSpaceBone {boneName=thigh_twist_01_r}
   }
   n29 = Anim Dynamics - Bone: dyn_top_r_strap(ComponentPose=@n28, Alpha=1, ExternalForce=@n25) {
    bGravityOverrideInSimSpace=true
    alphaBoolBlend {blendOption=Linear}
    boundBone {boneName=dyn_top_r_strap}
    componentAppliedLinearAccClamp=<500 500 500>
    componentLinearAccScale=<0.75 0.75 0.75>
    componentLinearVelScale=<0.75 0.75 0.75>
    physicsBodyDefinitions [{
      sphereCollisionRadius=0
      boundBone {boneName=dyn_top_r_strap}
      boxExtents=<8 3 3>
      constraintSetup {angularLimitsMax=<10 10 20>, angularLimitsMin=<-10 -10 -20>, angularTarget=<0 0 0>}
      localJointOffset=<2 0 0>
     }]
    relativeSpaceBone {boneName=thigh_twist_01_r}
   }
   n30 = Anim Dynamics - Bone: dyn_collar_strap(ComponentPose=@n29, Alpha=1, ExternalForce="0,0,0") {
    angularDampingOverride=1
    bGravityOverrideInSimSpace=true
    bOverrideAngularDamping=true
    bOverrideLinearDamping=true
    linearDampingOverride=1
    simulationSpace=RootRelative
    alphaBoolBlend {blendOption=Linear}
    boundBone {boneName=dyn_collar_strap}
    componentAppliedLinearAccClamp=<500 500 500>
    componentLinearAccScale=<0.25 0.25 0.25>
    componentLinearVelScale=<0.25 0.25 0.25>
    physicsBodyDefinitions [{
      sphereCollisionRadius=0
      boundBone {boneName=dyn_collar_strap}
      boxExtents=<10 5 5>
      constraintSetup {angularLimitsMax=<0 0 30>, angularLimitsMin=<0 0 -30>, angularTarget=<0 0 0>}
      localJointOffset=<2 0 0>
     }]
    relativeSpaceBone {boneName=head}
   }
   Output Pose(Result=Component To Local(ComponentPose=@n30)) {layerGroup=None}
  ]}
 EventGraph {nodes [
   Event BlueprintUpdateAnimation()
   > n1 = Sequence!(execute) -> (then_0, then_1, then_2:Exec)
   n2 = BreakTransform(InTransform=GetDeltaTransformFromRefPose(self=GetOwnerSkeletalMeshComponent(), BoneName=head, BaseName=spine_05))
   Set head_rot_back(execute=@n1.then_1, head_rot_back:double=MapRangeClamped(Value=BreakRotator(InRot=@n2.Rotation).Yaw, InRangeA=0, InRangeB=-55)) -> (Output_Get:double)
   n3 = Get bIsTargeting()
   > Set ads_rotate(ads_rotate:double="To Float (Boolean)"(InBool=@n3)) -> (Output_Get:double)
   > Set rotation2(rotation2:double="To Float (Boolean)"(InBool=NOT Boolean(A=@n3))) -> (Output_Get:double)
   Delay(execute=[@n9.then, @n1.then_0], Duration=0.25)
   > Set ext_force(ext_force:Vector="0,0,1250") -> (Output_Get:Vector)
   > n4 = Delay(Duration=0.375)
   Do Once(execute=@n9.else, Reset=@n4.then)
   > Set ext_force(ext_force="0, 0, 0")
   n5 = Get bIsSkydiveDiveMode()
   n6 = Get bIsSkydiveFloating()
   n7 = Get bIsSkyDiving()
   n8 = Get bIsParachuteOpened()
   > Set chain_alpha(chain_alpha:double=SelectFloat(A=0.25, B=0, bPickA=OR Boolean!(A=@n7, B=@n8, C:bool=@n6, D:bool=@n5) -> (ReturnValue))) -> (Output_Get:double)
   Set crouch_leg_volume(execute=@n1.then_2, crouch_leg_volume:double=SelectFloat(A=1, B=0, bPickA=Get bIsCrouching())) -> (Output_Get:double)
   n9 = Branch(Condition=Get bIsJumping())
   TryGetPawnOwner()
  ]}
]
variables [
 ads_rotate {type=number}
 chain_alpha {type=number}
 crouch_leg_volume {type=number}
 ext_force {type=Vector}
 head_rot_back {type=number}
 rotation2 {type=number}
]
