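# Physics asset for the Vigilante medium male character's hair mesh, setting up simulated motion for the top hair strands.
# Small capsule bodies on each dynamic hair bone sit alongside one large capsule on the head.
# Every hair body is constrained directly to the head rather than chained to its neighbours.
# Swing and twist limits are narrow, with some strands held noticeably stiffer than the rest.
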PhysicsAsset /BRCosmetics/Characters/Player/Male/Medium/Bodies/M_MED_Vigilante/Meshes/Parts/M_MED_Vigilante_Hair_Physics
bodies {
 dyn_hair_strand_top_01 [{shapeType=Capsule, length=4.0444, radius=1.5211, center=<3.6709 0.1216 -0.8884>, rotation=<rot 0 -85.1599 -65.7385>}]
 dyn_hair_strand_top_03 [{shapeType=Capsule, length=4.9555, radius=1.1461, center=<2.6209 0.1439 -1.5085>, rotation=<rot 0 -90.916 -62.4519>}]
 dyn_hair_strand_top_05 [{shapeType=Capsule, length=3.2044, radius=1.4763, center=<2.4521 -0.402 -0.3569>, rotation=<rot 0 -93.8517 -77.5928>}]
 dyn_hair_strand_top_07 [{shapeType=Capsule, length=3.5588, radius=1.2819, center=<1.921 -0.2589 -0.5724>, rotation=<rot 0 -97.8111 -71.2106>}]
 dyn_hair_strand_top_09 [{shapeType=Capsule, length=3.0067, radius=1.4352, center=<2.3313 0.1462 -0.3563>, rotation=<rot 0 -87.6783 -88.605>}]
 dyn_hair_strand_top_11 [{shapeType=Capsule, length=3.3066, radius=1.3671, rotation=<rot 0 92.5182 -87.7616>}]
 dyn_hair_strand_top_15 [{shapeType=Capsule, length=3.4282, radius=1.3051, center=<1.4975 0.3142 0.2699>, rotation=<rot 0 98.5086 -87.6483>}]
 dyn_hair_strand_top_17 [{shapeType=Capsule, length=3.4114, radius=1.32, center=<1.8737 0.3681 -0.3679>, rotation=<rot 0 -82.4746 -84.14>}]
 dyn_hair_strand_top_19 [{shapeType=Capsule, length=3.3924, radius=1.1946, center=<2.0762 0.1188 -0.4252>, rotation=<rot 0 -87.3975 -84.534>}]
 dyn_hair_strand_top_21 [{shapeType=Capsule, length=4.4802, radius=1.2137, center=<2.6697 0.3624 -1.3446>, rotation=<rot 0 -85.7061 -69.3751>}]
 dyn_hair_strand_top_23 [{shapeType=Capsule, length=4.7965, radius=1.3346, center=<2.9186 -0.4961 -1.7377>, rotation=<rot 0 -95.6084 -58.4903>}]
 dyn_hair_strand_top_25 [{shapeType=Capsule, length=4.0294, radius=1.1137, center=<3.8618 -0.3038 -1.0971>, rotation=<rot 0 -96.6657 -74.4362>}]
 dyn_hair_top_07 [{shapeType=Capsule, length=3.25, radius=1.255, center=<6.3922 -0.8222 0.2883>, rotation=<rot 0 -89.9996 89.5496>}]
 dyn_hair_top_08 [{shapeType=Capsule, length=3.2947, radius=1.3701, center=<7.3594 -0.4355 -0.5386>, rotation=<rot 0 -93.3867 -85.8223>}]
 head [{shapeType=Capsule, length=14.25, radius=8, center=<-2.4648 1.1009 -0.0523>, rotation=<rot 88.8196 -179.9992 146.5958>}]
}
constraints [
 {bone1Name=dyn_hair_strand_top_25, bone2Name=head, swing1LimitDegrees=20, swing2LimitDegrees=20, twistLimitDegrees=3}
 {bone1Name=dyn_hair_strand_top_23, bone2Name=head, swing1LimitDegrees=20, swing2LimitDegrees=20, twistLimitDegrees=3}
 {bone1Name=dyn_hair_strand_top_21, bone2Name=head, swing1LimitDegrees=20, swing2LimitDegrees=20, twistLimitDegrees=3}
 {bone1Name=dyn_hair_strand_top_19, bone2Name=head, swing1LimitDegrees=20, swing2LimitDegrees=20, twistLimitDegrees=3}
 {bone1Name=dyn_hair_strand_top_17, bone2Name=head, swing1LimitDegrees=5, swing2LimitDegrees=5, twistLimitDegrees=3}
 {bone1Name=dyn_hair_strand_top_15, bone2Name=head, swing1LimitDegrees=5, swing2LimitDegrees=5, twistLimitDegrees=3}
 {bone1Name=dyn_hair_top_08, bone2Name=head, swing1LimitDegrees=5, swing2LimitDegrees=5, twistLimitDegrees=3}
 {bone1Name=dyn_hair_top_07, bone2Name=head, swing1LimitDegrees=5, swing2LimitDegrees=5, twistLimitDegrees=3}
 {bone1Name=dyn_hair_strand_top_09, bone2Name=head, swing1LimitDegrees=20, swing2LimitDegrees=20, twistLimitDegrees=3}
 {bone1Name=dyn_hair_strand_top_11, bone2Name=head, swing1LimitDegrees=15, swing2LimitDegrees=15, twistLimitDegrees=5}
 {bone1Name=dyn_hair_strand_top_07, bone2Name=head, swing1LimitDegrees=20, swing2LimitDegrees=20, twistLimitDegrees=3}
 {bone1Name=dyn_hair_strand_top_05, bone2Name=head, swing1LimitDegrees=20, swing2LimitDegrees=20, twistLimitDegrees=3}
 {bone1Name=dyn_hair_strand_top_03, bone2Name=head, swing1LimitDegrees=20, swing2LimitDegrees=20, twistLimitDegrees=3}
 {bone1Name=dyn_hair_strand_top_01, bone2Name=head, swing1LimitDegrees=20, swing2LimitDegrees=20, twistLimitDegrees=3}
]
