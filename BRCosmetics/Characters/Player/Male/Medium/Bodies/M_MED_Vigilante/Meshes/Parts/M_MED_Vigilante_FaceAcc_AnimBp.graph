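# Anim Blueprint for the male Vigilante's face accessory mesh, adding secondary hair motion to the pose copied from its attached parent.
# An AnimDynamics chain simulates individual hair bones into a cached pose.
# That pose feeds two RigidBody nodes with separate FaceAcc and Hair physics assets, blended by pawn speed or jumping.
# The event graph enables partial simulation weights while jumping, falling or targeting; top hair bones keep fixed weights.
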AnimBlueprint /BRCosmetics/Characters/Player/Male/Medium/Bodies/M_MED_Vigilante/Meshes/Parts/M_MED_Vigilante_FaceAcc_AnimBp
target_skeleton=&/M_MED_Vigilante_FaceAcc_Skeleton
graphs [
 AnimGraph {nodes [
   n1 = RigidBody(ComponentPose=Local To Component(LocalPose=Use cached pose 'speedcache'()), Alpha=Get hair_alpha() -> (hair_alpha:double)) {
    bClampLinearTranslationLimitToRefPose=true
    bOverrideWorldGravity=true
    overridePhysicsAsset=&/M_MED_Vigilante_FaceAcc_Physics
    alphaBoolBlend {blendOption=Linear}
    baseBoneRef {boneName=neck_01}
    componentAppliedLinearAccClamp=<600 600 600>
    componentLinearAccScale=<0.25 0.25 0.25>
    componentLinearVelScale=<0.25 0.25 0.25>
    overrideWorldGravity {z=-500}
   }
   n2 = RigidBody(ComponentPose=Local To Component(LocalPose=Use cached pose 'speedcache'())) {
    bClampLinearTranslationLimitToRefPose=true
    bOverrideWorldGravity=true
    overridePhysicsAsset=&/M_MED_Vigilante_Hair_Physics
    alphaBoolBlend {blendOption=Linear}
    baseBoneRef {boneName=neck_01}
    componentAppliedLinearAccClamp=<600 600 600>
    componentLinearAccScale=<0.25 0.25 0.25>
    componentLinearVelScale=<0.25 0.25 0.25>
    overrideWorldGravity {z=-500}
   }
   n3 = Get PawnSpeed2D()
   n4 = Blend Poses by bool(
    BlendPose_0=Component To Local(ComponentPose=@n1)
    BlendPose_1=Component To Local(ComponentPose=@n2)
    BlendTime_0=0.2
    bActiveValue=OR Boolean!(
     A="float >= float"(A=@n3, B=411) {is=title K2Node_CallFunction}
     B="Equal (Float)"!(A:double=@n3, B:double=1) -> (ReturnValue:bool) {is=title K2Node_CallFunction}
     C:bool=Get bIsJumping()
    ) -> (ReturnValue)
   ) {blendType=Linear}
   Output Pose(Result=Component To Local(ComponentPose=Local To Component(LocalPose=@n4))) {layerGroup=None}
   n5 = Get hair_alpha()
   n6 = Anim Dynamics - Bone: dyn_hair_bottom_01(ComponentPose=Local To Component(LocalPose=Copy Pose From Mesh() {bUseAttachedParent=true}), Alpha=@n5, ExternalForce="0,0,0") {
    angularBiasOverride=1
    angularDampingOverride=1
    bGravityOverrideInSimSpace=true
    bOverrideAngularDamping=true
    bOverrideLinearDamping=true
    gravityScale=0.5
    simulationSpace=RootRelative
    alphaBoolBlend {blendOption=Linear}
    boundBone {boneName=dyn_hair_bottom_01}
    componentAppliedLinearAccClamp=<6000 6000 6000>
    componentLinearAccScale=<1 1 1>
    componentLinearVelScale=<1 1 1>
    physicsBodyDefinitions [{
      sphereCollisionRadius=0
      boundBone {boneName=dyn_hair_bottom_01}
      boxExtents=<10 14 6>
      constraintSetup {angularLimitsMax=<10 10 20>, angularLimitsMin=<-10 -10 -20>, angularTarget=<0 0 0>}
      localJointOffset=<6 0 0>
     }]
    relativeSpaceBone {boneName=neck_02}
   }
   n7 = Anim Dynamics - Bone: dyn_hair_bottom_02(ComponentPose=@n6, Alpha=@n5, ExternalForce="0,0,0") {
    angularBiasOverride=1
    angularDampingOverride=1
    bGravityOverrideInSimSpace=true
    bOverrideAngularDamping=true
    bOverrideLinearDamping=true
    gravityScale=0.5
    simulationSpace=RootRelative
    alphaBoolBlend {blendOption=Linear}
    boundBone {boneName=dyn_hair_bottom_02}
    componentAppliedLinearAccClamp=<6000 6000 6000>
    componentLinearAccScale=<1 1 1>
    componentLinearVelScale=<1 1 1>
    physicsBodyDefinitions [{
      sphereCollisionRadius=0
      boundBone {boneName=dyn_hair_bottom_02}
      boxExtents=<10 14 10>
      constraintSetup {angularLimitsMax=<10 10 20>, angularLimitsMin=<-10 -10 -20>, angularTarget=<0 0 0>}
      localJointOffset=<6 0 0>
     }]
    relativeSpaceBone {boneName=neck_02}
   }
   n8 = Anim Dynamics - Bone: dyn_hair_strand_bottom_03(ComponentPose=@n7, Alpha=@n5, ExternalForce="0,0,0") {
    angularBiasOverride=1
    angularDampingOverride=1
    bGravityOverrideInSimSpace=true
    bOverrideAngularDamping=true
    bOverrideLinearDamping=true
    gravityScale=0.5
    simulationSpace=RootRelative
    alphaBoolBlend {blendOption=Linear}
    boundBone {boneName=dyn_hair_strand_bottom_03}
    componentAppliedLinearAccClamp=<6000 6000 6000>
    componentLinearAccScale=<1 1 1>
    componentLinearVelScale=<1 1 1>
    physicsBodyDefinitions [{
      sphereCollisionRadius=0
      boundBone {boneName=dyn_hair_strand_bottom_03}
      boxExtents=<8 3 3>
      constraintSetup {angularLimitsMax=<10 10 20>, angularLimitsMin=<-10 -10 -20>, angularTarget=<0 0 0>}
      localJointOffset=<2 0 0>
     }]
    relativeSpaceBone {boneName=neck_02}
   }
   n9 = Anim Dynamics - Bone: dyn_hair_strand_bottom_13(ComponentPose=@n8, Alpha=@n5, ExternalForce="0,0,0") {
    angularBiasOverride=1
    angularDampingOverride=1
    bGravityOverrideInSimSpace=true
    bOverrideAngularDamping=true
    bOverrideLinearDamping=true
    gravityScale=0.5
    simulationSpace=RootRelative
    alphaBoolBlend {blendOption=Linear}
    boundBone {boneName=dyn_hair_strand_bottom_13}
    componentAppliedLinearAccClamp=<6000 6000 6000>
    componentLinearAccScale=<1 1 1>
    componentLinearVelScale=<1 1 1>
    physicsBodyDefinitions [{
      sphereCollisionRadius=0
      boundBone {boneName=dyn_hair_strand_bottom_13}
      boxExtents=<8 3 3>
      constraintSetup {angularLimitsMax=<10 10 20>, angularLimitsMin=<-10 -10 -20>, angularTarget=<0 0 0>}
      localJointOffset=<2 0 0>
     }]
    relativeSpaceBone {boneName=neck_02}
   }
   n10 = Anim Dynamics - Bone: dyn_hair_strand_bottom_15(ComponentPose=@n9, Alpha=@n5, ExternalForce="0,0,0") {
    angularBiasOverride=1
    angularDampingOverride=1
    bGravityOverrideInSimSpace=true
    bOverrideAngularDamping=true
    bOverrideLinearDamping=true
    gravityScale=0.5
    simulationSpace=RootRelative
    alphaBoolBlend {blendOption=Linear}
    boundBone {boneName=dyn_hair_strand_bottom_15}
    componentAppliedLinearAccClamp=<6000 6000 6000>
    componentLinearAccScale=<1 1 1>
    componentLinearVelScale=<1 1 1>
    physicsBodyDefinitions [{
      sphereCollisionRadius=0
      boundBone {boneName=dyn_hair_strand_bottom_15}
      boxExtents=<8 3 3>
      constraintSetup {angularLimitsMax=<10 10 20>, angularLimitsMin=<-10 -10 -20>, angularTarget=<0 0 0>}
      localJointOffset=<2 0 0>
     }]
    relativeSpaceBone {boneName=neck_02}
   }
   n11 = Get hair_alpha_half() -> (hair_alpha_half:double)
   n12 = Anim Dynamics - Bone: dyn_hair_strand_bottom_05(ComponentPose=@n10, Alpha=@n11, ExternalForce="0,0,0") {
    angularBiasOverride=1
    angularDampingOverride=1
    bGravityOverrideInSimSpace=true
    bOverrideAngularDamping=true
    bOverrideLinearDamping=true
    gravityScale=0.5
    simulationSpace=RootRelative
    alphaBoolBlend {blendOption=Linear}
    boundBone {boneName=dyn_hair_strand_bottom_05}
    componentAppliedLinearAccClamp=<6000 6000 6000>
    componentLinearAccScale=<1 1 1>
    componentLinearVelScale=<1 1 1>
    physicsBodyDefinitions [{
      sphereCollisionRadius=0
      boundBone {boneName=dyn_hair_strand_bottom_05}
      boxExtents=<8 3 3>
      constraintSetup {angularLimitsMax=<10 10 20>, angularLimitsMin=<-10 -10 -20>, angularTarget=<0 0 0>}
      localJointOffset=<2 0 0>
     }]
    relativeSpaceBone {boneName=neck_02}
   }
   n13 = Anim Dynamics - Bone: dyn_hair_strand_bottom_07(ComponentPose=@n12, Alpha=@n11, ExternalForce="0,0,0") {
    angularBiasOverride=1
    angularDampingOverride=1
    bGravityOverrideInSimSpace=true
    bOverrideAngularDamping=true
    bOverrideLinearDamping=true
    gravityScale=0.5
    simulationSpace=RootRelative
    alphaBoolBlend {blendOption=Linear}
    boundBone {boneName=dyn_hair_strand_bottom_07}
    componentAppliedLinearAccClamp=<6000 6000 6000>
    componentLinearAccScale=<1 1 1>
    componentLinearVelScale=<1 1 1>
    physicsBodyDefinitions [{
      sphereCollisionRadius=0
      boundBone {boneName=dyn_hair_strand_bottom_07}
      boxExtents=<8 3 3>
      constraintSetup {angularLimitsMax=<10 10 20>, angularLimitsMin=<-10 -10 -20>, angularTarget=<0 0 0>}
      localJointOffset=<2 0 0>
     }]
    relativeSpaceBone {boneName=neck_02}
   }
   n14 = Anim Dynamics - Bone: dyn_hair_strand_bottom_09(ComponentPose=@n13, Alpha=@n11, ExternalForce="0,0,0") {
    angularBiasOverride=1
    angularDampingOverride=1
    bGravityOverrideInSimSpace=true
    bOverrideAngularDamping=true
    bOverrideLinearDamping=true
    gravityScale=0.5
    simulationSpace=RootRelative
    alphaBoolBlend {blendOption=Linear}
    boundBone {boneName=dyn_hair_strand_bottom_09}
    componentAppliedLinearAccClamp=<6000 6000 6000>
    componentLinearAccScale=<1 1 1>
    componentLinearVelScale=<1 1 1>
    physicsBodyDefinitions [{
      sphereCollisionRadius=0
      boundBone {boneName=dyn_hair_strand_bottom_09}
      boxExtents=<8 3 3>
      constraintSetup {angularLimitsMax=<10 10 20>, angularLimitsMin=<-10 -10 -20>, angularTarget=<0 0 0>}
      localJointOffset=<4 0 0>
     }]
    relativeSpaceBone {boneName=neck_02}
   }
   n15 = Anim Dynamics - Bone: dyn_hair_bottom_07(ComponentPose=@n14, Alpha=@n11, ExternalForce="0,0,0") {
    angularBiasOverride=1
    angularDampingOverride=1
    bGravityOverrideInSimSpace=true
    bOverrideAngularDamping=true
    bOverrideLinearDamping=true
    gravityScale=0.5
    simulationSpace=RootRelative
    alphaBoolBlend {blendOption=Linear}
    boundBone {boneName=dyn_hair_bottom_07}
    componentAppliedLinearAccClamp=<9000 9000 9000>
    componentLinearAccScale=<1 1 1>
    componentLinearVelScale=<1 1 1>
    physicsBodyDefinitions [{
      sphereCollisionRadius=0
      boundBone {boneName=dyn_hair_bottom_07}
      boxExtents=<10 8 8>
      constraintSetup {angularLimitsMax=<5 5 10>, angularLimitsMin=<-5 -5 -10>, angularTarget=<0 0 0>}
      localJointOffset=<4 0 0>
     }]
    relativeSpaceBone {boneName=neck_02}
   }
   n16 = Anim Dynamics - Bone: dyn_hair_bottom_08(ComponentPose=@n15, Alpha=@n11, ExternalForce="0,0,0") {
    angularBiasOverride=1
    angularDampingOverride=1
    bGravityOverrideInSimSpace=true
    bOverrideAngularDamping=true
    bOverrideLinearDamping=true
    gravityScale=0.5
    simulationSpace=RootRelative
    alphaBoolBlend {blendOption=Linear}
    boundBone {boneName=dyn_hair_bottom_08}
    componentAppliedLinearAccClamp=<9000 9000 9000>
    componentLinearAccScale=<1 1 1>
    componentLinearVelScale=<1 1 1>
    physicsBodyDefinitions [{
      sphereCollisionRadius=0
      boundBone {boneName=dyn_hair_bottom_08}
      boxExtents=<10 8 8>
      constraintSetup {angularLimitsMax=<5 5 10>, angularLimitsMin=<-5 -5 -10>, angularTarget=<0 0 0>}
      localJointOffset=<4 0 0>
     }]
    relativeSpaceBone {boneName=neck_02}
   }
   n17 = Anim Dynamics - Bone: dyn_hair_top_08(ComponentPose=@n16, Alpha=0.25, ExternalForce="0,0,0") {
    angularBiasOverride=1
    angularDampingOverride=1
    bGravityOverrideInSimSpace=true
    bOverrideAngularDamping=true
    bOverrideLinearDamping=true
    gravityScale=0.5
    simulationSpace=RootRelative
    alphaBoolBlend {blendOption=Linear}
    boundBone {boneName=dyn_hair_top_08}
    componentAppliedLinearAccClamp=<6000 6000 6000>
    componentLinearAccScale=<1 1 1>
    componentLinearVelScale=<1 1 1>
    physicsBodyDefinitions [{
      sphereCollisionRadius=0
      boundBone {boneName=dyn_hair_top_08}
      boxExtents=<10 3 3>
      constraintSetup {angularLimitsMax=<10 10 20>, angularLimitsMin=<-10 -10 -20>, angularTarget=<0 0 0>}
      localJointOffset=<2 0 0>
     }]
    relativeSpaceBone {boneName=neck_02}
   }
   n18 = Anim Dynamics - Bone: dyn_hair_top_07(ComponentPose=@n17, Alpha=0.5, ExternalForce="0,0,0") {
    angularBiasOverride=1
    angularDampingOverride=1
    bGravityOverrideInSimSpace=true
    bOverrideAngularDamping=true
    bOverrideLinearDamping=true
    gravityScale=0.5
    simulationSpace=RootRelative
    alphaBoolBlend {blendOption=Linear}
    boundBone {boneName=dyn_hair_top_07}
    componentAppliedLinearAccClamp=<6000 6000 6000>
    componentLinearAccScale=<1 1 1>
    componentLinearVelScale=<1 1 1>
    physicsBodyDefinitions [{
      sphereCollisionRadius=0
      boundBone {boneName=dyn_hair_top_07}
      boxExtents=<10 3 3>
      constraintSetup {angularLimitsMax=<10 10 20>, angularLimitsMin=<-10 -10 -20>, angularTarget=<0 0 0>}
      localJointOffset=<2 0 0>
     }]
    relativeSpaceBone {boneName=neck_02}
   }
   Save cached pose 'speedcache'(Pose=Component To Local(ComponentPose=@n18)) {cacheName=speedcache}
  ]}
 EventGraph {nodes [
   Event BlueprintUpdateAnimation()
   n1 = SelectFloat(A=0.5, B=0, bPickA=OR Boolean(A=Get bIsJumping(), B=Get bIsFalling(), C=Get bIsTargeting()))
   > Set hair_alpha(hair_alpha:double=@n1) -> (Output_Get:double)
   > Set hair_alpha_half(hair_alpha_half:double=float * float(A=@n1, B=0.5)) -> (Output_Get:double)
   TryGetPawnOwner()
   Get bIsSkyDiving()
   Get bIsSkydiveDiveMode()
   Get bIsSkydiveFloating()
   Get bIsParachuteOpened()
  ]}
]
variables [
 hair_alpha {type=number}
 hair_alpha_half {type=number}
]
